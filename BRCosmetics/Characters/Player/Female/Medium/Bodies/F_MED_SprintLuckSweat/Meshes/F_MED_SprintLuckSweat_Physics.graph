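# Physics asset for the F_MED_SprintLuckSweat female player body, built for the dynamic bones of its long coat.
# Capsules on legs, feet, pelvis and lower spine, plus a huge root sphere reaching just above the origin, provide body shapes around seven coat bone chains.
# Chain links have tight angular limits, mostly widening down the chain.
# Cross constraints with locked swing and twist tie neighbouring chains together lower down.
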PhysicsAsset /BRCosmetics/Characters/Player/Female/Medium/Bodies/F_MED_SprintLuckSweat/Meshes/F_MED_SprintLuckSweat_Physics
bodies {
 calf_l [{shapeType=Capsule, length=32.9252, radius=8.129, shapeName=calf_r_capsule, center=<-21.412 -2.437 -0.5377>, rotation=<rot 84.4204 -180 -180>}]
 calf_r [{shapeType=Capsule, length=32.9252, radius=8.129, shapeName=calf_r_capsule, center=<21.412 2.47 0.5377>, rotation=<rot 84.9325 -180 -180>}]
 dyn_coat_a_l_01 [{shapeType=Capsule, length=20.531, radius=6.0091, shapeName=dyn_coat_a_l_01_capsule, center=<7.6058 0 6.0295>, rotation=<rot 0 0 90>}]
 dyn_coat_a_l_02 [{shapeType=Capsule, length=21.4036, radius=5.6771, shapeName=dyn_coat_a_l_02_capsule, center=<7.3666 0 4.2885>, rotation=<rot 0 0 90>}]
 dyn_coat_a_l_03 [{shapeType=Capsule, length=26.1529, radius=5.7936, shapeName=dyn_coat_a_l_03_capsule, center=<6.8787 0 1.861>, rotation=<rot 0 0 90>}]
 dyn_coat_a_l_04 [{shapeType=Capsule, length=26.878, radius=5.7936, shapeName=dyn_coat_a_l_04_capsule, center=<6.4369 0 0.9102>, rotation=<rot 0 0 90>}]
 dyn_coat_a_l_MSTR [{shapeType=Capsule, length=0, radius=0, shapeName=dyn_coat_a_l_MSTR_capsule, rotation=<rot 0 -90 0>}]
 dyn_coat_a_r_01 [{shapeType=Capsule, length=13.0812, radius=6.1029, shapeName=dyn_coat_a_r_01_capsule, center=<7.6058 0 6.3279>, rotation=<rot 0 0 90>}]
 dyn_coat_a_r_02 [{
   shapeType=Capsule
   length=20.5636
   radius=6.3456
   shapeName=dyn_coat_a_r_02_capsule
   center=<5.3644 0.1039 4.8039>
   rotation=<rot 0 0 82.949>
  }]
 dyn_coat_a_r_03 [{
   shapeType=Capsule
   length=22.3563
   radius=6.7872
   shapeName=dyn_coat_a_r_03_capsule
   center=<2.9919 0.1315 3.8813>
   rotation=<rot 0 0 82.949>
  }]
 dyn_coat_a_r_04 [{
   shapeType=Capsule
   length=22.3081
   radius=7.4747
   shapeName=dyn_coat_a_r_04_capsule
   center=<3.1947 0.0722 3.3847>
   rotation=<rot 0 0 82.949>
  }]
 dyn_coat_a_r_MSTR [{shapeType=Capsule, length=0, radius=0, shapeName=dyn_coat_a_r_MSTR_capsule, rotation=<rot 0 -90 0>}]
 dyn_coat_bk_01 [{shapeType=Capsule, length=16.7996, radius=6.9094, shapeName=dyn_coat_bk_01_capsule, center=<7.9139 0 6.3334>, rotation=<rot 0 0 90>}]
 dyn_coat_bk_02 [{shapeType=Capsule, length=20.6213, radius=9.063, shapeName=dyn_coat_bk_02_capsule, center=<3.8671 0 5.1112>, rotation=<rot 0 0 90>}]
 dyn_coat_bk_03 [{shapeType=Capsule, length=21.3049, radius=8.3794, shapeName=dyn_coat_bk_03_capsule, center=<2.5253 0 4.2958>, rotation=<rot 0 0 90>}]
 dyn_coat_bk_04 [{shapeType=Capsule, length=20.619, radius=9.0652, shapeName=dyn_coat_bk_04_capsule, center=<1.6534 0 3.9993>, rotation=<rot 0 0 90>}]
 dyn_coat_bk_MSTR [{shapeType=Capsule, length=0, radius=0, shapeName=dyn_coat_bk_MSTR_capsule, rotation=<rot 0 -90 0>}]
 dyn_coat_bk_l_01 [{
   shapeType=Capsule
   length=27.6238
   radius=5.5119
   shapeName=dyn_coat_bk_l_01_capsule
   center=<7.6058 0.16 4.2879>
   rotation=<rot 0 0 84.3335>
  }]
 dyn_coat_bk_l_02 [{
   shapeType=Capsule
   length=26.4276
   radius=6.7081
   shapeName=dyn_coat_bk_l_02_capsule
   center=<4.6681 0.0603 3.454>
   rotation=<rot 0 0 84.3335>
  }]
 dyn_coat_bk_l_03 [{shapeType=Capsule, length=26.5973, radius=6.5384, shapeName=dyn_coat_bk_l_03_capsule, center=<2.9898 0 4.2303>, rotation=<rot 0 0 84.3335>}]
 dyn_coat_bk_l_04 [{shapeType=Capsule, length=23.7045, radius=9.4311, shapeName=dyn_coat_bk_l_04_capsule, center=<2.5803 0 2.6776>, rotation=<rot 0 0 84.3335>}]
 dyn_coat_bk_l_MSTR [{shapeType=Capsule, length=0, radius=0, shapeName=dyn_coat_bk_l_MSTR_capsule, rotation=<rot 0 -90 0>}]
 dyn_coat_bk_r_01 [{shapeType=Capsule, length=26.3125, radius=5.7936, shapeName=dyn_coat_bk_r_01_capsule, center=<7.6058 0 5.4354>, rotation=<rot 0 0 90>}]
 dyn_coat_bk_r_02 [{shapeType=Capsule, length=25.0891, radius=7.017, shapeName=dyn_coat_bk_r_02_capsule, center=<4.6776 0 4.9663>, rotation=<rot -15.5541 0 90>}]
 dyn_coat_bk_r_03 [{shapeType=Capsule, length=25.4835, radius=6.6226, shapeName=dyn_coat_bk_r_03_capsule, center=<3.1044 0 5.5736>, rotation=<rot 0 0 90>}]
 dyn_coat_bk_r_04 [{
   shapeType=Capsule
   length=23.8856
   radius=8.2205
   shapeName=dyn_coat_bk_r_04_capsule
   center=<2.5803 -0.0751 3.0672>
   rotation=<rot 0 0 92.9154>
  }]
 dyn_coat_bk_r_MSTR [{shapeType=Capsule, length=0, radius=0, shapeName=dyn_coat_bk_r_MSTR_capsule, rotation=<rot 0 -90 0>}]
 dyn_coat_l_01 [{
   shapeType=Capsule
   length=5.91
   radius=5.0492
   shapeName=dyn_coat_l_01_capsule
   center=<5.0855 -2.5209 3.1692>
   rotation=<rot 2.2103 -9.6894 78.0966>
  }]
 dyn_coat_l_02 [{
   shapeType=Capsule
   length=7.8236
   radius=5.3322
   shapeName=dyn_coat_l_02_capsule
   center=<5.964 -2.6369 3.2882>
   rotation=<rot 0.4575 1.4413 76.8903>
  }]
 dyn_coat_l_03 [{
   shapeType=Capsule
   length=18.5323
   radius=6.3756
   shapeName=dyn_coat_l_03_capsule
   center=<10.6557 -6.4006 4.4909>
   rotation=<rot 0 0 93.6281>
  }]
 dyn_coat_l_04 [{shapeType=Capsule, length=19.2774, radius=6.7601, shapeName=dyn_coat_l_04_capsule, center=<9.5048 -5.8899 5.1612>, rotation=<rot 0 0 91.989>}]
 dyn_coat_l_05 [{
   shapeType=Capsule
   length=18.2059
   radius=7.7968
   shapeName=dyn_coat_l_05_capsule
   center=<7.7145 -5.5784 4.5262>
   rotation=<rot -0.098 2.0382 89.9315>
  }]
 dyn_coat_l_06 [{shapeType=Capsule, length=20.1958, radius=8.383, shapeName=dyn_coat_l_06_capsule, center=<4.5958 -5.8673 6.3424>, rotation=<rot 0 0 89.412>}]
 dyn_coat_l_MSTR [{shapeType=Capsule, length=0, radius=0, shapeName=dyn_coat_l_MSTR_capsule, rotation=<rot 0 -90 0>}]
 dyn_coat_r_01 [{shapeType=Capsule, length=5.91, radius=5.0492, shapeName=dyn_coat_r_01_capsule, center=<5.0643 0 3.7329>, rotation=<rot 0 0 90>}]
 dyn_coat_r_02 [{shapeType=Capsule, length=5.91, radius=5.0492, shapeName=dyn_coat_r_02_capsule, center=<5.0643 -0.0546 2.8707>, rotation=<rot 0 0 98.7936>}]
 dyn_coat_r_03 [{
   shapeType=Capsule
   length=16.3464
   radius=6.561
   shapeName=dyn_coat_r_03_capsule
   center=<6.6615 2.8586 5.2778>
   rotation=<rot 0.062 3.8066 77.2405>
  }]
 dyn_coat_r_04 [{
   shapeType=Capsule
   length=20.1677
   radius=7.8998
   shapeName=dyn_coat_r_04_capsule
   center=<7.193 1.8826 6.7421>
   rotation=<rot 0.099 0.1591 71.9282>
  }]
 dyn_coat_r_05 [{
   shapeType=Capsule
   length=22.0256
   radius=7.6687
   shapeName=dyn_coat_r_05_capsule
   center=<7.1891 2.2089 4.5591>
   rotation=<rot -0.6562 -1.0089 71.1058>
  }]
 dyn_coat_r_06 [{
   shapeType=Capsule
   length=22.3252
   radius=7.0545
   shapeName=dyn_coat_r_06_capsule
   center=<7.133 4.1272 2.5773>
   rotation=<rot 0.0001 -0.0001 71.1015>
  }]
 dyn_coat_r_MSTR [{shapeType=Capsule, length=0, radius=0, shapeName=dyn_coat_r_MSTR_capsule, rotation=<rot 0 -90 0>}]
 foot_l [{
   shapeType=Capsule
   length=11.5202
   radius=11.6018
   shapeName=foot_l_capsule
   center=<-3.7581 4.3599 -0.0837>
   rotation=<rot 0 20.6525 -89.3033>
  }]
 foot_r [{
   shapeType=Capsule
   length=11.5202
   radius=11.6018
   shapeName=foot_r_capsule
   center=<3.7581 -4.3599 0.0837>
   rotation=<rot 0 20.6525 -89.3033>
  }]
 pelvis [{shapeType=Capsule, length=10.4206, radius=11.1598, shapeName=pelvis_capsule, center=<-6.0952 -2.3301 0>}]
 root [{shapeType=Sphere, radius=195.0485, shapeName=root_sphere, center=<0 0 -189.3105>}]
 spine_01 [{shapeType=Capsule, length=8.8934, radius=7.9021, shapeName=spine_01_capsule, center=<0 0.2928 0>, rotation=<rot 0 -90 0>}]
 spine_02 [{shapeType=Capsule, length=4.3104, radius=9.2193, shapeName=spine_02_capsule, center=<2.0312 0.9308 0>}]
 thigh_l [
  {
   shapeType=Capsule
   length=5.3882
   radius=8.2566
   shapeName=thigh_r_capsule
   center=<-9.223 -3.6755 0.346>
   rotation=<rot 78.6944 -176.8284 -158.7946>
  }
  {
   shapeType=Capsule
   length=27.7178
   radius=8.6117
   shapeName=thigh_r_capsule
   center=<-28.3862 -0.7755 1.4318>
   rotation=<rot 81.59 -10.7406 -13.1143>
  }
  {
   shapeType=Capsule
   length=36.1468
   radius=8.1447
   shapeName=thigh_r_capsule
   center=<-22.3367 -1.7093 2.4717>
   rotation=<rot 87.8922 -23.6702 -21.2163>
  }
  {
   shapeType=Capsule
   length=7.6284
   radius=8.2566
   shapeName=thigh_r_capsule
   center=<-9.1126 0.0345 0.163>
   rotation=<rot 79.3423 -173.419 -168.6736>
  }
 ]
 thigh_r [
  {
   shapeType=Capsule
   length=5.3882
   radius=8.2566
   shapeName=thigh_r_capsule
   center=<9.223 3.6755 -0.346>
   rotation=<rot 78.6944 -176.8284 -158.7946>
  }
  {
   shapeType=Capsule
   length=27.7178
   radius=8.6117
   shapeName=thigh_r_capsule
   center=<28.3862 0.7755 -1.4318>
   rotation=<rot 81.59 -10.7406 -11.9335>
  }
  {
   shapeType=Capsule
   length=36.1468
   radius=8.1447
   shapeName=thigh_r_capsule
   center=<22.3367 1.7093 -2.4717>
   rotation=<rot 87.8922 -23.6702 -21.2163>
  }
  {
   shapeType=Capsule
   length=13.0873
   radius=8.2566
   shapeName=thigh_r_capsule
   center=<7.9371 0.0671 -0.3806>
   rotation=<rot 79.3423 -173.419 -168.6736>
  }
 ]
}
constraints [
 {bone1Name=dyn_coat_a_l_01, bone2Name=dyn_coat_a_l_MSTR, swing1LimitDegrees=10, swing2LimitDegrees=15, twistLimitDegrees=5}
 {bone1Name=dyn_coat_a_l_02, bone2Name=dyn_coat_a_l_01, swing1LimitDegrees=10, swing2LimitDegrees=20, twistLimitDegrees=5}
 {bone1Name=dyn_coat_a_l_03, bone2Name=dyn_coat_a_l_02, swing1LimitDegrees=10, swing2LimitDegrees=25, twistLimitDegrees=5}
 {bone1Name=dyn_coat_a_l_04, bone2Name=dyn_coat_a_l_03, swing1LimitDegrees=10, swing2LimitDegrees=30, twistLimitDegrees=5}
 {bone1Name=dyn_coat_a_r_01, bone2Name=dyn_coat_a_r_MSTR, swing1LimitDegrees=10, swing2LimitDegrees=15, twistLimitDegrees=5}
 {bone1Name=dyn_coat_a_r_02, bone2Name=dyn_coat_a_r_01, swing1LimitDegrees=10, swing2LimitDegrees=20, twistLimitDegrees=5}
 {bone1Name=dyn_coat_a_r_03, bone2Name=dyn_coat_a_r_02, swing1LimitDegrees=10, swing2LimitDegrees=25, twistLimitDegrees=5}
 {bone1Name=dyn_coat_a_r_04, bone2Name=dyn_coat_a_r_03, swing1LimitDegrees=10, swing2LimitDegrees=30, twistLimitDegrees=5}
 {bone1Name=dyn_coat_bk_01, bone2Name=dyn_coat_bk_MSTR, swing1LimitDegrees=10, swing2LimitDegrees=15, twistLimitDegrees=5}
 {bone1Name=dyn_coat_bk_02, bone2Name=dyn_coat_bk_01, swing1LimitDegrees=10, swing2LimitDegrees=20, twistLimitDegrees=5}
 {bone1Name=dyn_coat_bk_03, bone2Name=dyn_coat_bk_02, swing1LimitDegrees=10, swing2LimitDegrees=25, twistLimitDegrees=5}
 {bone1Name=dyn_coat_bk_04, bone2Name=dyn_coat_bk_03, swing1LimitDegrees=10, swing2LimitDegrees=30, twistLimitDegrees=5}
 {bone1Name=dyn_coat_bk_l_01, bone2Name=dyn_coat_bk_l_MSTR, swing1LimitDegrees=10, swing2LimitDegrees=15, twistLimitDegrees=5}
 {bone1Name=dyn_coat_bk_l_02, bone2Name=dyn_coat_bk_l_01, swing1LimitDegrees=10, swing2LimitDegrees=20, twistLimitDegrees=5}
 {bone1Name=dyn_coat_bk_l_03, bone2Name=dyn_coat_bk_l_02, swing1LimitDegrees=10, swing2LimitDegrees=25, twistLimitDegrees=5}
 {bone1Name=dyn_coat_bk_l_04, bone2Name=dyn_coat_bk_l_03, swing1LimitDegrees=10, swing2LimitDegrees=30, twistLimitDegrees=5}
 {bone1Name=dyn_coat_bk_r_01, bone2Name=dyn_coat_bk_r_MSTR, swing1LimitDegrees=10, swing2LimitDegrees=15, twistLimitDegrees=5}
 {bone1Name=dyn_coat_bk_r_02, bone2Name=dyn_coat_bk_r_01, swing1LimitDegrees=10, swing2LimitDegrees=20, twistLimitDegrees=5}
 {bone1Name=dyn_coat_bk_r_03, bone2Name=dyn_coat_bk_r_02, swing1LimitDegrees=10, swing2LimitDegrees=25, twistLimitDegrees=5}
 {bone1Name=dyn_coat_bk_r_04, bone2Name=dyn_coat_bk_r_03, swing1LimitDegrees=10, swing2LimitDegrees=30, twistLimitDegrees=5}
 {bone1Name=dyn_coat_r_02, bone2Name=dyn_coat_r_01, swing1LimitDegrees=5, swing2LimitDegrees=25, twistLimitDegrees=5}
 {bone1Name=dyn_coat_r_03, bone2Name=dyn_coat_r_02, swing1LimitDegrees=5, swing2LimitDegrees=25, twistLimitDegrees=5}
 {bone1Name=dyn_coat_r_04, bone2Name=dyn_coat_r_03, swing1LimitDegrees=5, swing2LimitDegrees=25, twistLimitDegrees=5}
 {bone1Name=dyn_coat_r_05, bone2Name=dyn_coat_r_04, swing1LimitDegrees=5, swing2LimitDegrees=25, twistLimitDegrees=5}
 {bone1Name=dyn_coat_r_06, bone2Name=dyn_coat_r_05, swing1LimitDegrees=5, swing2LimitDegrees=25, twistLimitDegrees=5}
 {bone1Name=dyn_coat_l_02, bone2Name=dyn_coat_l_01, swing1LimitDegrees=5, swing2LimitDegrees=10, twistLimitDegrees=5}
 {bone1Name=dyn_coat_l_03, bone2Name=dyn_coat_l_02, swing1LimitDegrees=5, swing2LimitDegrees=15, twistLimitDegrees=5}
 {bone1Name=dyn_coat_l_04, bone2Name=dyn_coat_l_03, swing1LimitDegrees=5, swing2LimitDegrees=20, twistLimitDegrees=5}
 {bone1Name=dyn_coat_l_05, bone2Name=dyn_coat_l_04, swing1LimitDegrees=5, swing2LimitDegrees=20, twistLimitDegrees=5}
 {bone1Name=dyn_coat_l_06, bone2Name=dyn_coat_l_05, swing1LimitDegrees=5, swing2LimitDegrees=25, twistLimitDegrees=5}
 {
  bone1Name=dyn_coat_l_05
  bone2Name=dyn_coat_a_l_03
  swing1LimitDegrees=45
  swing1Motion=Locked
  swing2LimitDegrees=45
  swing2Motion=Locked
  twistLimitDegrees=45
  twistMotion=Locked
 }
 {
  bone1Name=dyn_coat_a_r_03
  bone2Name=dyn_coat_r_05
  swing1LimitDegrees=45
  swing1Motion=Locked
  swing2LimitDegrees=45
  swing2Motion=Locked
  twistLimitDegrees=45
  twistMotion=Locked
 }
 {
  bone1Name=dyn_coat_bk_l_03
  bone2Name=dyn_coat_bk_03
  swing1LimitDegrees=45
  swing1Motion=Locked
  swing2LimitDegrees=45
  swing2Motion=Locked
  twistLimitDegrees=45
  twistMotion=Locked
 }
 {
  bone1Name=dyn_coat_bk_r_03
  bone2Name=dyn_coat_bk_03
  swing1LimitDegrees=45
  swing1Motion=Locked
  swing2LimitDegrees=45
  swing2Motion=Locked
  twistLimitDegrees=45
  twistMotion=Locked
 }
]
