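# Physics asset for the belt mesh of the medium female SprintLuckSweat player cosmetic body, covering two dangling belt chains, coat bones and the pelvis.
# Left and right belt chains run from zero-size MSTR anchor bodies through thin capsules, with tight swing and twist limits between links.
# Larger capsules, some paired per bone, cover the pelvis and coat.
# Coat joints mix locked and free angular motion, and no constraint joins the belts to coat or pelvis.
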PhysicsAsset /BRCosmetics/Characters/Player/Female/Medium/Bodies/F_MED_SprintLuckSweat/Meshes/F_MED_SprintLuckSweat_belt_Physics
bodies {
 dyn_belt_l_01 [{shapeType=Capsule, length=14.4008, radius=1.497, shapeName=dyn_belt_l_01_capsule, center=<8.2202 0 1.1224>, rotation=<rot 90 0 0>}]
 dyn_belt_l_02 [{shapeType=Capsule, length=14.4008, radius=1.497, shapeName=dyn_belt_l_02_capsule, center=<8.2202 0.049 1.0717>, rotation=<rot 81.8317 90 90>}]
 dyn_belt_l_03 [{
   shapeType=Capsule
   length=14.4008
   radius=1.497
   shapeName=dyn_belt_l_03_capsule
   center=<8.221 -0.0232 1.2178>
   rotation=<rot 80.4684 91.9741 92.0017>
  }]
 dyn_belt_l_04 [{shapeType=Capsule, length=10.9591, radius=1.497, shapeName=dyn_belt_l_04_capsule, center=<6.1316 0 0.2074>, rotation=<rot 80.2649 0 0>}]
 dyn_belt_l_MSTR [{shapeType=Capsule, length=0, radius=0, shapeName=dyn_belt_l_MSTR_capsule, rotation=<rot 0 -90 0>}]
 dyn_belt_r_01 [{shapeType=Capsule, length=9.3576, radius=1.497, shapeName=dyn_belt_r_01_capsule, center=<4.969 0 1.2222>, rotation=<rot 90 0 0>}]
 dyn_belt_r_02 [{shapeType=Capsule, length=9.3576, radius=1.497, shapeName=dyn_belt_r_02_capsule, center=<4.9762 0 1.0453>, rotation=<rot 87.6577 180 180>}]
 dyn_belt_r_03 [{shapeType=Capsule, length=9.3576, radius=1.497, shapeName=dyn_belt_r_03_capsule, center=<4.937 0 1.7065>, rotation=<rot 86.2151 180 180>}]
 dyn_belt_r_MSTR [{shapeType=Capsule, length=0, radius=0, shapeName=dyn_belt_r_MSTR_capsule, rotation=<rot 0 -90 0>}]
 dyn_coat_l_04 [
  {
   shapeType=Capsule
   length=37.5632
   radius=9.3207
   shapeName=dyn_coat_l_04_capsule
   center=<0.8528 -8.9924 -12.7734>
   rotation=<rot 71.4275 -88.8778 0.759>
  }
  {
   shapeType=Capsule
   length=20.056
   radius=6.9599
   shapeName=dyn_coat_l_04_capsule
   center=<0.7458 10.0388 -0.6115>
   rotation=<rot 0 0 -13.3856>
  }
 ]
 dyn_coat_l_05 [
  {
   shapeType=Capsule
   length=37.5632
   radius=9.3207
   shapeName=dyn_coat_l_05_capsule
   center=<2.1123 -8.8762 -12.6802>
   rotation=<rot 72.5524 -90 0>
  }
  {shapeType=Capsule, length=17.6166, radius=7.6482, shapeName=dyn_coat_l_05_capsule, center=<0 13.4572 1.4824>, rotation=<rot 0 0 -40.5646>}
 ]
 dyn_coat_l_06 [
  {
   shapeType=Capsule
   length=37.5632
   radius=9.3207
   shapeName=dyn_coat_l_06_capsule
   center=<2.7852 -8.5872 -12.9233>
   rotation=<rot 68.1967 -90.07 -0.0263>
  }
  {shapeType=Capsule, length=15.8552, radius=8.818, shapeName=dyn_coat_l_06_capsule, center=<1.5229 12.6939 5.7933>, rotation=<rot 0 0 -40.931>}
 ]
 dyn_coat_r_03 [{
   shapeType=Capsule
   length=40.7177
   radius=5.1411
   shapeName=dyn_coat_r_03_capsule
   center=<7.7915 10.8724 -3.765>
   rotation=<rot 0 0 84.3692>
  }]
 dyn_coat_r_04 [{
   shapeType=Capsule
   length=36.6529
   radius=7.8851
   shapeName=dyn_coat_r_04_capsule
   center=<2.0897 12.9121 -5.9413>
   rotation=<rot 85.2974 -90.0221 -1.0165>
  }]
 dyn_coat_r_05 [{
   shapeType=Capsule
   length=36.6529
   radius=7.8851
   shapeName=dyn_coat_r_05_capsule
   center=<3.3231 12.9042 -5.8452>
   rotation=<rot 85.2969 -90 0>
  }]
 dyn_coat_r_06 [{
   shapeType=Capsule
   length=36.6529
   radius=7.8851
   shapeName=dyn_coat_r_06_capsule
   center=<4.1999 12.8116 -4.7204>
   rotation=<rot 85.2969 -90 0>
  }]
 pelvis [{shapeType=Capsule, length=5.1511, radius=11.8648, shapeName=pelvis_capsule, center=<0 -0.1235 0.4644>}]
}
constraints [
 {
  bone1Name=dyn_coat_l_05
  bone2Name=dyn_coat_l_04
  swing1LimitDegrees=45
  swing1Motion=Locked
  swing2LimitDegrees=45
  swing2Motion=Locked
  twistLimitDegrees=45
  twistMotion=Locked
 }
 {
  bone1Name=dyn_coat_r_05
  bone2Name=dyn_coat_r_04
  swing1LimitDegrees=45
  swing1Motion=Locked
  swing2LimitDegrees=45
  swing2Motion=Locked
  twistLimitDegrees=45
  twistMotion=Locked
 }
 {
  bone1Name=dyn_coat_l_04
  bone2Name=pelvis
  swing1LimitDegrees=45
  swing1Motion=Free
  swing2LimitDegrees=45
  swing2Motion=Free
  twistLimitDegrees=45
  twistMotion=Free
 }
 {
  bone1Name=dyn_coat_l_06
  bone2Name=dyn_coat_l_05
  swing1LimitDegrees=45
  swing1Motion=Free
  swing2LimitDegrees=45
  swing2Motion=Free
  twistLimitDegrees=45
  twistMotion=Free
 }
 {
  bone1Name=dyn_coat_r_04
  bone2Name=dyn_coat_r_03
  swing1LimitDegrees=45
  swing1Motion=Free
  swing2LimitDegrees=45
  swing2Motion=Free
  twistLimitDegrees=45
  twistMotion=Free
 }
 {
  bone1Name=dyn_coat_r_06
  bone2Name=dyn_coat_r_05
  swing1LimitDegrees=45
  swing1Motion=Free
  swing2LimitDegrees=45
  swing2Motion=Free
  twistLimitDegrees=45
  twistMotion=Free
 }
 {bone1Name=dyn_belt_r_01, bone2Name=dyn_belt_r_MSTR, swing1LimitDegrees=10, swing2LimitDegrees=25, twistLimitDegrees=2}
 {bone1Name=dyn_belt_r_02, bone2Name=dyn_belt_r_01, swing1LimitDegrees=10, swing2LimitDegrees=25, twistLimitDegrees=2}
 {bone1Name=dyn_belt_r_03, bone2Name=dyn_belt_r_02, swing1LimitDegrees=10, swing2LimitDegrees=25, twistLimitDegrees=2}
 {bone1Name=dyn_belt_l_01, bone2Name=dyn_belt_l_MSTR, swing1LimitDegrees=10, swing2LimitDegrees=25, twistLimitDegrees=5}
 {bone1Name=dyn_belt_l_02, bone2Name=dyn_belt_l_01, swing1LimitDegrees=10, swing2LimitDegrees=25, twistLimitDegrees=5}
 {bone1Name=dyn_belt_l_03, bone2Name=dyn_belt_l_02, swing1LimitDegrees=10, swing2LimitDegrees=25, twistLimitDegrees=5}
 {bone1Name=dyn_belt_l_04, bone2Name=dyn_belt_l_03, swing1LimitDegrees=10, swing2LimitDegrees=25, twistLimitDegrees=5}
]
