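# Physics asset for the cloth belt on the F_MED_SprintLuckSweat female player cosmetic body, covering a dangling right-side belt chain.
# Three capsule belt bodies chain from a zero-size master body with tight swing and twist limits.
# Larger pelvis, right thigh and right calf capsules provide leg and hip volumes, with both leg bodies constrained directly to the pelvis with free rotation.
# No constraint links the belt master to the pelvis.
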PhysicsAsset /BRCosmetics/Characters/Player/Female/Medium/Bodies/F_MED_SprintLuckSweat/Meshes/F_MED_SprintLuckSweat_cloth_belt_Physics
bodies {
 calf_r [{shapeType=Capsule, length=23.5311, radius=14.6937, shapeName=calf_r_capsule, center=<21.412 1.8236 0>, rotation=<rot 90 0 0>}]
 dyn_belt_r_01 [{shapeType=Capsule, length=6.8108, radius=2.7903, shapeName=dyn_belt_r_01_capsule, center=<5.7118 0 2.6194>, rotation=<rot -79.8592 -180 180>}]
 dyn_belt_r_02 [{shapeType=Capsule, length=6.8608, radius=3.1012, shapeName=dyn_belt_r_02_capsule, center=<5.371 0 2.2699>, rotation=<rot -86.3858 0 0>}]
 dyn_belt_r_03 [{shapeType=Capsule, length=5.949, radius=3.1051, shapeName=dyn_belt_r_03_capsule, center=<5.3562 0 3.097>, rotation=<rot -87.1518 0 0>}]
 dyn_belt_r_MSTR [{shapeType=Capsule, length=0, radius=0, shapeName=dyn_belt_r_MSTR_capsule, rotation=<rot 0 -90 0>}]
 pelvis [{shapeType=Capsule, length=3.8501, radius=11.8648, shapeName=pelvis_capsule, center=<0 -0.1235 0.4644>}]
 thigh_r [{
   shapeType=Capsule
   length=32.6584
   radius=15.0359
   shapeName=thigh_r_capsule
   center=<25.5082 2.6361 -2.4556>
   rotation=<rot -75.6695 -25.1213 26.8629>
  }]
}
constraints [
 {bone1Name=dyn_belt_r_01, bone2Name=dyn_belt_r_MSTR, swing1LimitDegrees=10, swing2LimitDegrees=20, twistLimitDegrees=5}
 {bone1Name=dyn_belt_r_02, bone2Name=dyn_belt_r_01, swing1LimitDegrees=10, swing2LimitDegrees=20, twistLimitDegrees=5}
 {bone1Name=dyn_belt_r_03, bone2Name=dyn_belt_r_02, swing1LimitDegrees=10, swing2LimitDegrees=20, twistLimitDegrees=5}
 {
  bone1Name=thigh_r
  bone2Name=pelvis
  swing1LimitDegrees=45
  swing1Motion=Free
  swing2LimitDegrees=45
  swing2Motion=Free
  twistLimitDegrees=45
  twistMotion=Free
 }
 {
  bone1Name=calf_r
  bone2Name=pelvis
  swing1LimitDegrees=45
  swing1Motion=Free
  swing2LimitDegrees=45
  swing2Motion=Free
  twistLimitDegrees=45
  twistMotion=Free
 }
]
