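# Post-process style AnimBlueprint for the SprintLuckSweat female body cosmetic, adding secondary motion to ribbons, belts and skirt.
# Copied parent pose runs through Control Rig and ribbon AnimDynamics, then branches: cloth enabled uses constrained, movement-offset belts with RigidBody; disabled uses two chained RigidBodies.
# A front-end curve blends DefaultSlot onto the skirt.
# EventGraph drives Chaos cloth and gravity from a parameter asset, resetting cloth after customization and emotes.
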AnimBlueprint /BRCosmetics/Characters/Player/Female/Medium/Bodies/F_MED_SprintLuckSweat/Meshes/F_MED_SprintLuckSweat_AnimBP
target_skeleton=&/F_MED_SprintLuckSweat_Skeleton
graphs [
 AnimGraph {nodes [
   Output Pose(Result=Layered blend per bone(
     BasePose=Use cached pose 'StartPose'()
     BlendPoses_0=Slot 'DefaultSlot'()
     BlendWeights_0=Get Custom Idle Is Playing Float() -> (Custom Idle Is Playing Float:double)
    ) {layerSetup [{branchFilters [{boneName=dyn_skirt_f_m}]}]}) {layerGroup=None}
   n1 = "Transform (Modify) Bone - Bone: dyn_belt_r_MSTR"(
    ComponentPose=Local To Component(LocalPose=Use cached pose 'StartupCache_1'())
    Alpha=1
    Translation="-2,0,0"
    Rotation="-30,0,0"
    Scale="1,1,1"
   ) {rotationMode=BMM_Additive, translationMode=BMM_Additive, boneToModify {boneName=dyn_belt_r_MSTR}, rotation {pitch=-30}, translation {x=-2}}
   n2 = Use cached pose 'StartupCache_1'()
   n3 = Blend Poses by bool(
    BlendPose_0=Component To Local(ComponentPose=@n1)
    BlendPose_1=Component To Local(ComponentPose=Local To Component(LocalPose=@n2))
    BlendTime_0=0.2
    BlendTime_1=0.5
    bActiveValue=Get bIsMoving2D()
   ) {blendType=Linear}
   n4 = "Transform (Modify) Bone - Bone: dyn_belt_r_MSTR"(
    ComponentPose=Local To Component(LocalPose=Use cached pose 'StartupCache_1'())
    Alpha=1
    Translation="-5,0,0"
    Rotation="-20,0,0"
    Scale="1,1,1"
   ) {rotationMode=BMM_Additive, translationMode=BMM_Additive, boneToModify {boneName=dyn_belt_r_MSTR}, rotation {pitch=-20}, translation {x=-5}}
   n5 = Component To Local(ComponentPose=@n4)
   n6 = "Get Crouch moving "() -> ("Crouch moving ":bool)
   n7 = RigidBody!(
    ComponentPose=Local To Component(LocalPose=Blend Poses by bool(BlendPose_0=@n5, BlendPose_1=@n3, BlendTime_0=0.2, BlendTime_1=0.5, bActiveValue=@n6) {blendType=Linear})
    Alpha
    OverrideWorldGravity:Vector="0,0,0"
    ExternalForce
    SimSpaceSettings:Sim Space Settings Structure=(WorldAlpha=1,VelocityScaleZ=0.35,DampingAlpha=0.5,MaxLinearVelocity=770,MaxAngularVelocity=2,MaxLinearAcceleration=3000,MaxAngularAcceleration=28,ExternalLinearDragV=(X=0.1,Y=0.1,Z=0.1),ExternalLinearVelocity=(X=0,Y=0,Z=0),ExternalAngularVelocity=(X=0,Y=0,Z=0))
   ) -> (Pose) {
    bClampLinearTranslationLimitToRefPose=true
    bOverrideWorldGravity=true
    overridePhysicsAsset=&/F_MED_SprintLuckSweat_cloth_belt_Physics
    simulationSpace=BaseBoneSpace
    baseBoneRef {boneName=pelvis}
    simSpaceSettings {
     dampingAlpha=0.5
     maxAngularAcceleration=28
     maxAngularVelocity=2
     maxLinearAcceleration=3000
     maxLinearVelocity=770
     velocityScaleZ=0.35
     worldAlpha=1
     externalLinearDragV=<0.1 0.1 0.1>
    }
   }
   n8 = RigidBody(
    ComponentPose=Local To Component(LocalPose=Use cached pose 'StartupCache'())
    OverrideWorldGravity="0,0,0"
    SimSpaceSettings=(WorldAlpha=1,VelocityScaleZ=0.35,DampingAlpha=0.5,MaxLinearVelocity=770,MaxAngularVelocity=2,MaxLinearAcceleration=3000,MaxAngularAcceleration=28,ExternalLinearDragV=(X=0.1,Y=0.1,Z=0.1),ExternalLinearVelocity=(X=0,Y=0,Z=0),ExternalAngularVelocity=(X=0,Y=0,Z=0))
   ) {
    bClampLinearTranslationLimitToRefPose=true
    bOverrideWorldGravity=true
    overridePhysicsAsset=&/F_MED_SprintLuckSweat_Physics
    simulationSpace=BaseBoneSpace
    baseBoneRef {boneName=pelvis}
    simSpaceSettings {
     dampingAlpha=0.5
     maxAngularAcceleration=28
     maxAngularVelocity=2
     maxLinearAcceleration=3000
     maxLinearVelocity=770
     velocityScaleZ=0.35
     worldAlpha=1
     externalLinearDragV=<0.1 0.1 0.1>
    }
   }
   n9 = RigidBody(
    ComponentPose=@n8
    OverrideWorldGravity="0,0,0"
    SimSpaceSettings=(WorldAlpha=1,VelocityScaleZ=0.35,DampingAlpha=0.5,MaxLinearVelocity=770,MaxAngularVelocity=2,MaxLinearAcceleration=3000,MaxAngularAcceleration=28,ExternalLinearDragV=(X=0.1,Y=0.1,Z=0.1),ExternalLinearVelocity=(X=0,Y=0,Z=0),ExternalAngularVelocity=(X=0,Y=0,Z=0))
   ) {
    bClampLinearTranslationLimitToRefPose=true
    bOverrideWorldGravity=true
    overridePhysicsAsset=&/F_MED_SprintLuckSweat_belt_Physics
    simulationSpace=BaseBoneSpace
    baseBoneRef {boneName=pelvis}
    simSpaceSettings {
     dampingAlpha=0.5
     maxAngularAcceleration=28
     maxAngularVelocity=2
     maxLinearAcceleration=3000
     maxLinearVelocity=770
     velocityScaleZ=0.35
     worldAlpha=1
     externalLinearDragV=<0.1 0.1 0.1>
    }
   }
   Save cached pose 'StartPose'(Pose=Blend Poses by bool(
     BlendPose_0=Component To Local(ComponentPose=@n7)
     BlendPose_1=Component To Local(ComponentPose=@n9)
     BlendTime_0=0.2
     BlendTime_1=0.5
     bActiveValue=Get bClothEnabled()
    ) {blendType=Linear}) {cacheName=StartPose}
   n10 = "Constraint - dyn_belt_l_MSTR (PPPPPPP)"!(
    ComponentPose=Local To Component(LocalPose=Use cached pose 'StartupCache'())
    Alpha=1
    ConstraintWeights_0=0.2847
    ConstraintWeights_1:float=0.2475
    ConstraintWeights_2:float=0.172
    ConstraintWeights_3:float=0.1607
    ConstraintWeights_4:float=0.0629
    ConstraintWeights_5:float=0.0568
    ConstraintWeights_6:float=0.0154
   ) -> (Pose) {
    boneToModify {boneName=dyn_belt_l_MSTR}
    constraintSetup [
     {targetBone {boneName=dyn_coat_l_02}}
     {targetBone {boneName=dyn_coat_l_03}}
     {targetBone {boneName=pelvis}}
     {targetBone {boneName=dyn_coat_a_l_01}}
     {targetBone {boneName=spine_01}}
     {targetBone {boneName=spine_02}}
     {targetBone {boneName=spine_04}}
    ]
    constraintWeights [1,1,1,1,1,1,1]
   }
   n11 = "Constraint - dyn_belt_r_MSTR (PPPPPPP)"!(
    ComponentPose=@n10
    Alpha=1
    ConstraintWeights_0=0.3396
    ConstraintWeights_1:float=0.2766
    ConstraintWeights_2:float=0.1576
    ConstraintWeights_3:float=0.1401
    ConstraintWeights_4:float=0.0474
    ConstraintWeights_5:float=0.0332
    ConstraintWeights_6:float=0.0055
   ) -> (Pose) {
    boneToModify {boneName=dyn_belt_r_MSTR}
    constraintSetup [
     {targetBone {boneName=dyn_coat_r_02}}
     {targetBone {boneName=dyn_coat_r_03}}
     {targetBone {boneName=pelvis}}
     {targetBone {boneName=dyn_coat_a_r_01}}
     {targetBone {boneName=spine_01}}
     {targetBone {boneName=spine_02}}
     {targetBone {boneName=spine_04}}
    ]
    constraintWeights [1,1,1,1,1,1,1]
   }
   Save cached pose 'StartupCache_1'(Pose=Component To Local(ComponentPose=@n11)) {cacheName=StartupCache_1}
   n12 = Control Rig!(
    Source=Copy Pose From Mesh() {bUseAttachedParent=true}
    Alpha
    "is Crouch Moving ":bool="Get Crouch moving "()
    "is Moving ":bool=Get bIsMoving2D()
    is sliding:bool=Get bIsKneeSliding()
    "is Crouching ":bool=Get bIsCrouching()
    Is Cloth Enabled:bool=Get bClothEnabled()
   ) -> (Pose) {controlRigAssetReference {blueprintRigClass=&/F_MED_SprintLuckSweat_CtrlRig.F_MED_SprintLuckSweat_CtrlRig_C}}
   n13 = Anim Dynamics - Bone: dyn_ribbon_r_01(ComponentPose=Local To Component(LocalPose=@n12), Alpha=1, ExternalForce="0,0,0") {
    angularDampingOverride=0.99
    angularSpringConstant=5000
    bAngularSpring=true
    bOverrideAngularDamping=true
    bUseGravityOverride=true
    gravityScale=0.35
    simulationSpace=BoneRelative
    boundBone {boneName=dyn_ribbon_r_01}
    componentAppliedLinearAccClamp=<0 0 0>
    gravityOverride {z=-980}
    physicsBodyDefinitions [{
      boundBone {boneName=dyn_ribbon_r_01}
      boxExtents=<20 15 15>
      constraintSetup {angularLimitsMax=<2 3 10>, angularLimitsMin=<-2 -10 -10>, angularTarget=<1 0 0>}
      localJointOffset=<4 0 0>
     }]
    relativeSpaceBone {boneName=lowerarm_twist_01_r}
    simSpaceSettings {maxAngularAcceleration=64, maxAngularVelocity=16, simSpaceAngularAlpha=1}
   }
   n14 = Anim Dynamics - Bone: dyn_libbon_l_01(ComponentPose=@n13, Alpha=1, ExternalForce="0,0,0") {
    angularDampingOverride=0.99
    angularSpringConstant=5000
    bAngularSpring=true
    bOverrideAngularDamping=true
    bUseGravityOverride=true
    gravityScale=0.35
    simulationSpace=BoneRelative
    boundBone {boneName=dyn_libbon_l_01}
    componentAppliedLinearAccClamp=<0 0 0>
    gravityOverride {z=-980}
    physicsBodyDefinitions [{
      boundBone {boneName=dyn_libbon_l_01}
      boxExtents=<20 15 15>
      constraintSetup {angularLimitsMax=<2 3 10>, angularLimitsMin=<-2 -10 -10>, angularTarget=<1 0 0>}
      localJointOffset=<4 0 0>
     }]
    relativeSpaceBone {boneName=lowerarm_twist_01_l}
    simSpaceSettings {maxAngularAcceleration=64, maxAngularVelocity=16, simSpaceAngularAlpha=1}
   }
   Save cached pose 'StartupCache'(Pose=Component To Local(ComponentPose=@n14)) {cacheName=StartupCache}
  ]}
 EventGraph {nodes [
   n1 = Event BlueprintUpdateAnimation()
   > Set DeltaTimeLocal(DeltaTimeLocal:double=@n1.DeltaTimeX) -> (Output_Get:double)
   > n2 = Sequence!(execute) -> (then_0, then_1, then_2:Exec, then_3:Exec)
   > "Set Crouch moving "("Crouch moving ":bool=AND Boolean(A=Get bIsCrouching(), B=Get bIsMoving2D())) -> (Output_Get:bool)
   n3 = Branch(execute=@n2.then_1, Condition=Get bClothEnabled())
   > n4 = Sequence!(execute) -> (then_0, then_1)
   > n5 = "Cloth Controller Function (Parameter Asset) (Nativized)"!(
    execute
    CharacterPartAnimInstance:Custom Character Part Anim Instance Object Reference=Self-Reference()
    ClothingAssetName:Name=SprintLuckSweat_Cloth
    ClothCoreSettings:"Cloth Core Settings Structure (by ref)"=Get Cloth Core Settings Runtime() -> (Cloth Core Settings Runtime:Cloth Core Settings Structure)
    WindGustLocalSettings:"Wind Gust Settings Structure (by ref)"=Get Wind Gust Local Settings() -> (Wind Gust Local Settings:Wind Gust Settings Structure)
    ClothUpAxis:Vector="1,0,0"
    LinVelScaleFwdBwdClamp:float=0
    LinVelScaleUpDwnClamp:float=0.7
    DeltaTime:float=Get DeltaTimeLocal() -> (DeltaTimeLocal:double)
    GameTimeInSeconds:float=GetGameTimeInSeconds()
    StateLogicLayerName:Name=none
    Parameters:Fort Character Dynamics Parameters Object Reference=&/F_MED_SprintLuckSweat_Parameters
   ) -> (
    then
    OutCurrentState:Name=None
    OutClothParameters:Cloth Parameters Structure
    OutClothCoreSettings:Cloth Core Settings Structure
    OutWindGustSettings:Wind Gust Parameters Structure
    OutWindGustLocalSettings:Wind Gust Settings Structure
    OutWindVelocity:Vector
    OutLinearVelVector:Vector
   )
   > Set Wind Gust Local Settings(Wind Gust Local Settings:Wind Gust Settings Structure=@n5.OutWindGustLocalSettings) -> (Output_Get:Wind Gust Settings Structure)
   n6 = "Gravity Override State Function (Parameter Asset) (Nativized)"(
    execute=[@n4.then_1, @n3.else]
    CharacterPartAnimInstance=Self-Reference()
    GameTimeInSeconds=GetGameTimeInSeconds()
    Parameters=&F_MED_SprintLuckSweat_Parameters
   )
   > Set RBAN Settings Out_Native(RBAN Settings Out_Native:Fort Rigid Body Anim Node Parameters Structure=@n6.OutRigidBodyAnimNodeParameters) -> (Output_Get:Fort Rigid Body Anim Node Parameters Structure)
   > Set GravityOverride(GravityOverride=@n6)
   Branch(execute=@n2.then_2, Condition=Get bIsInFrontEnd())
   > Set Custom Idle Is Playing Float(Custom Idle Is Playing Float:double=GetCurveValue(self=GetOwnerFrontEndAnimInstance(), CurveName=Frontend_Idle_Bypass)) -> (Output_Get:double)
   n7 = Event OnCharacterCustomizationLoaded()
   > Set Custom Base Mesh(Custom Base Mesh:Skeletal Mesh Component Object Reference=@n7.CustomBaseMesh) -> (Output_Get:Skeletal Mesh Component Object Reference)
   n8 = CheckForBackbling()
   > Set isWearingBackbling(isWearingBackbling=OR Boolean(A=@n8, B=@n8.bIsBackblingLong))
   > Set Is Backbling A Cape(Is Backbling A Cape:bool=@n8.bIsBackblingACape) -> (Output_Get:bool)
   > DelayUntilNextFrame!(execute) -> (then)
   n9 = GetOwningComponent()
   > n10 = Cast To ChaosClothingInteractor!(
    execute
    Object=GetClothingInteractor!(
     self:Clothing Simulation Interactor Object Reference=GetClothingSimulationInteractor!(self:Skeletal Mesh Component Object Reference=@n9, ClothingAssetName:Name=None) -> (ReturnValue:Clothing Simulation Interactor Object Reference)
     ClothingAssetName:Name=SprintLuckSweat_Cloth
    ) -> (ReturnValue:Clothing Interactor Object Reference)
   ) -> (then, CastFailed, AsChaos Clothing Interactor:Chaos Clothing Interactor Object Reference) {is=title K2Node_DynamicCast}
   > ResetAndTeleport!(execute, self:Chaos Clothing Interactor Object Reference=@n10, bReset:bool=true, bTeleport:bool=true) -> (then)
   Event OnEndPlayingEmote() {is=title K2Node_Event}
   > Set Cloth Core Settings Runtime(Cloth Core Settings Runtime:Cloth Core Settings Structure=Get Cloth Core Settings Default() -> (Cloth Core Settings Default:Cloth Core Settings Structure)) -> (Output_Get:Cloth Core Settings Structure)
   > Branch(Condition=Get bClothEnabled())
   n11 = GetOwningComponent()
   > n12 = Cast To ChaosClothingInteractor(Object=GetClothingInteractor(self=GetClothingSimulationInteractor(self=@n11, ClothingAssetName=None), ClothingAssetName=SprintLuckSweat_Cloth))
   > ResetAndTeleport(self=@n12, bReset=true, bTeleport=false)
  ]}
]
variables [
 Cloth Core Settings Default {
  type=ClothCoreSettings
  default {
   collisionThickness=1.5
   dampingCoefficient=0.08
   frictionCoefficient=0.3
   localDampingCoefficient=0
   numberIterations=3
   numberSubsteps=1
   selfCollisionThickness=0
   animDriveDamping=<0 1>
   animDriveStiffness=<0 1>
   areaStiffness=<1 1>
   bendingStiffness=<0.9 1>
   edgeStiffness=<0.9 1>
   tetherStiffness=<1 1>
  }
 }
 Cloth Core Settings Runtime {
  type=ClothCoreSettings
  default {
   collisionThickness=1.5
   dampingCoefficient=0.08
   frictionCoefficient=0.3
   localDampingCoefficient=0
   numberIterations=2
   numberSubsteps=1
   selfCollisionThickness=0
   animDriveDamping=<0 1>
   animDriveStiffness=<0 1>
   areaStiffness=<1 1>
   bendingStiffness=<1 1>
   edgeStiffness=<1 1>
   tetherStiffness=<1 1>
  }
 }
 "Crouch moving " {type=boolean}
 Custom Base Mesh {type=SkeletalMeshComponent}
 Custom Idle Is Playing Float {type=number}
 Delta Time X {type=number}
 DeltaTimeLocal {type=number}
 GravityOverride {type=Vector, default=<0 0 980>}
 Is Backbling A Cape {type=boolean}
 isWearingBackbling {type=boolean}
 RBAN Settings Out_Native {type=FortRigidBodyAnimNodeParameters}
 Wind Gust Local Settings {type=WindGustSettings, default {elapsedTime=0, finalGustStrengthLocal=0, gustDuration=0, gustStrength=0, timeBetweenGusts=0}}
]
