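# Cloth collider physics asset for the SprintLuckSweat medium female player body cosmetic, covering the lower torso and upper legs.
# Capsules wrap the pelvis, two lower spine bones, a right dynamic belt bone and both thighs, which use mirrored overlapping pairs.
# A wide, flat box on the root sits just below the origin like a floor.
# Calf and foot bodies have no shapes, and no constraints link the bodies.
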PhysicsAsset /BRCosmetics/Characters/Player/Female/Medium/Bodies/F_MED_SprintLuckSweat/Meshes/F_MED_SprintLuckSweat_ClothColliders
bodies {
 calf_l []
 calf_r []
 dyn_belt_r_01 [{
   shapeType=Capsule
   length=6.4617
   radius=2.2762
   shapeName=dyn_belt_r_01_capsule
   center=<4.8524 0.0083 2.0547>
   rotation=<rot 1.3037 -100.9515 -91.2246>
  }]
 foot_l []
 foot_r []
 pelvis [{
   shapeType=Capsule
   length=7.9329
   radius=10.8254
   shapeName=pelvis_capsule
   center=<-5.6651 -3.0215 0.4128>
   rotation=<rot 0 -77.3778 1.0064>
  }]
 root [{shapeType=Box, extentX=200, extentY=200, extentZ=20.809, shapeName=root_box, center=<0 0 -10.928>, rotation=<rot 0 -90 0>}]
 spine_01 [{shapeType=Capsule, length=4.8547, radius=9.5718, center=<-2.8278 1.4447 0>}]
 spine_02 [{shapeType=Capsule, length=2.5677, radius=8.6272, shapeName=spine_02_capsule, center=<0 1.8073 0.0071>, rotation=<rot -0.224 -90 1.4151>}]
 thigh_l [
  {
   shapeType=Capsule
   length=5.3882
   radius=8.2566
   shapeName=thigh_r_capsule
   center=<-9.2216 -3.6713 0.3463>
   rotation=<rot 78.6944 -176.8284 -158.7946>
  }
  {
   shapeType=Capsule
   length=13.8001
   radius=7.5438
   shapeName=thigh_r_capsule
   center=<-7.9368 -0.0627 0.3807>
   rotation=<rot 79.3423 -173.419 -168.6736>
  }
 ]
 thigh_r [
  {
   shapeType=Capsule
   length=5.3882
   radius=8.2566
   shapeName=thigh_r_capsule
   center=<9.2216 3.6713 -0.3463>
   rotation=<rot 78.6944 -176.8284 -158.7946>
  }
  {
   shapeType=Capsule
   length=13.8001
   radius=7.5438
   shapeName=thigh_r_capsule
   center=<7.9368 0.0627 -0.3807>
   rotation=<rot 79.3423 -173.419 -168.6736>
  }
 ]
}
constraints []
