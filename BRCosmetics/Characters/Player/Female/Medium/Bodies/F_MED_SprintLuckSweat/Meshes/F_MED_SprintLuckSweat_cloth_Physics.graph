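# Physics asset for the cloth mesh of the SprintLuckSweat medium female player body cosmetic, covering the dynamic bone chain on the right side of the coat.
# Three capsule bodies run along the coat bones, plus a tiny capsule on the master bone.
# Two constraints link the three segments in sequence with tight swing and twist limits.
# No constraint joins the master body to the chain.
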PhysicsAsset /BRCosmetics/Characters/Player/Female/Medium/Bodies/F_MED_SprintLuckSweat/Meshes/F_MED_SprintLuckSweat_cloth_Physics
bodies {
 dyn_coat_r_01 [{shapeType=Capsule, length=5.91, radius=5.0492, shapeName=dyn_coat_r_01_capsule, center=<5.0643 0 -3.9591>, rotation=<rot 0 0 90>}]
 dyn_coat_r_02 [{shapeType=Capsule, length=5.91, radius=5.0492, shapeName=dyn_coat_r_02_capsule, center=<6.5336 0 0>, rotation=<rot 90 0 0>}]
 dyn_coat_r_03 [{shapeType=Capsule, length=8.4129, radius=5.0492, shapeName=dyn_coat_r_03_capsule, center=<6.5336 0 0>, rotation=<rot 90 0 0>}]
 dyn_coat_r_MSTR [{shapeType=Capsule, length=0.505, radius=0.505, shapeName=dyn_coat_r_MSTR_capsule, rotation=<rot 0 -90 0>}]
}
constraints [
 {bone1Name=dyn_coat_r_02, bone2Name=dyn_coat_r_01, swing1LimitDegrees=10, swing2LimitDegrees=20, twistLimitDegrees=5}
 {bone1Name=dyn_coat_r_03, bone2Name=dyn_coat_r_02, swing1LimitDegrees=10, swing2LimitDegrees=20, twistLimitDegrees=5}
]
