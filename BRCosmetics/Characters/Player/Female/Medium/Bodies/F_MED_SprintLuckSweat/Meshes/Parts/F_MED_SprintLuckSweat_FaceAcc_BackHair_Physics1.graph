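# Physics asset for the back hair of the SprintLuckSweat medium female character's face accessory part.
# Left, middle and right three-bone capsule chains hang from zero-size base bodies linked to spine_05, with very tight angular limits along each.
# Angularly locked cross-links tie the outer chains' second segments to the middle one, and a right-side corrective body has looser limits.
# Upper spine, clavicle and upper-arm capsules give the body's shape, joined by angularly free constraints.
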PhysicsAsset /BRCosmetics/Characters/Player/Female/Medium/Bodies/F_MED_SprintLuckSweat/Meshes/Parts/F_MED_SprintLuckSweat_FaceAcc_BackHair_Physics1
bodies {
 base_backHair_L_01 [{shapeType=Capsule, length=0, radius=0, shapeName=base_backHair_L_01_capsule, rotation=<rot 0 -90 0>}]
 base_backHair_M_01 [{shapeType=Capsule, length=0, radius=0, shapeName=base_backHair_M_01_capsule, rotation=<rot 0 -90 0>}]
 base_backHair_R_01 [{shapeType=Capsule, length=0, radius=0, shapeName=base_backHair_R_01_capsule, rotation=<rot 0 -90 0>}]
 clavicle_l [{
   shapeType=Capsule
   length=8.8433
   radius=6.1545
   shapeName=clavicle_l_capsule
   center=<3.745 -1.0162 0.1235>
   rotation=<rot 77.3884 3.5917 2.8621>
  }]
 clavicle_r [{
   shapeType=Capsule
   length=8.7705
   radius=6.2273
   shapeName=clavicle_l_capsule
   center=<-5.0277 0.4158 -0.2499>
   rotation=<rot 72.3458 1.6301 1.7484>
  }]
 dyn_backHair_L_01 [{shapeType=Capsule, length=7.2654, radius=3.7599, shapeName=dyn_backHair_L_01_capsule, center=<5.6657 -1.5839 0>, rotation=<rot 90 0 0>}]
 dyn_backHair_L_02 [{shapeType=Capsule, length=5.8556, radius=3.7812, shapeName=dyn_backHair_L_02_capsule, center=<5.6657 -3.1689 0>, rotation=<rot 90 0 0>}]
 dyn_backHair_L_03 [{
   shapeType=Capsule
   length=4.0709
   radius=3.7273
   shapeName=dyn_backHair_L_03_capsule
   center=<0.4088 -1.7464 -0.008>
   rotation=<rot 66.9222 -24.9551 -8.1134>
  }]
 dyn_backHair_M_01 [{shapeType=Capsule, length=7.2654, radius=3.7599, shapeName=dyn_backHair_M_01_capsule, center=<5.6657 -1.5839 0>, rotation=<rot 90 0 0>}]
 dyn_backHair_M_02 [{shapeType=Capsule, length=5.8556, radius=3.7812, shapeName=dyn_backHair_M_02_capsule, center=<5.6657 -3.9486 0>, rotation=<rot 90 0 0>}]
 dyn_backHair_M_03 [{
   shapeType=Capsule
   length=4.7527
   radius=3.7273
   shapeName=dyn_backHair_M_03_capsule
   center=<1.8241 -2.6337 0>
   rotation=<rot 79.3098 -65.6815 -102.2661>
  }]
 dyn_backHair_R_01 [{
   shapeType=Capsule
   length=7.2654
   radius=3.7599
   shapeName=dyn_backHair_R_01_capsule
   center=<5.8406 -0.5929 0>
   rotation=<rot 90 -16.7811 0>
  }]
 dyn_backHair_R_02 [{shapeType=Capsule, length=5.8556, radius=3.7812, shapeName=dyn_backHair_R_02_capsule, center=<5.6657 -2.4935 0>}]
 dyn_backHair_R_03 [{
   shapeType=Capsule
   length=4.7527
   radius=3.7273
   shapeName=dyn_backHair_R_03_capsule
   center=<4.612 -1.4945 0>
   rotation=<rot 79.3098 -81.0676 -102.2661>
  }]
 dyn_backHair_R_corrective [{
   shapeType=Capsule
   length=4.7099
   radius=3.3978
   shapeName=dyn_backHair_R_corrective_capsule
   center=<0.466 -0.9269 -0.271>
   rotation=<rot -1.3173 -136.4668 -51.8583>
  }]
 spine_03 [{shapeType=Capsule, length=5.4934, radius=9.3962, shapeName=spine_03_capsule, center=<-2.0838 1.5304 0>}]
 spine_04 [{shapeType=Capsule, length=9.5996, radius=9.8801, shapeName=spine_04_capsule, center=<1.1773 1.1088 0>}]
 spine_05 [{shapeType=Capsule, length=9.8774, radius=9.2226, shapeName=spine_05_capsule, center=<2.7418 0.8639 0>}]
 upperarm_l [{
   shapeType=Capsule
   length=24.0782
   radius=5.9218
   shapeName=upperarm_l_capsule
   center=<12.6605 0.2945 -0.0207>
   rotation=<rot -86.8515 0.0017 -0.0016>
  }]
 upperarm_r [{
   shapeType=Capsule
   length=23.8534
   radius=6.1465
   shapeName=upperarm_l_capsule
   center=<-12.7438 -0.2556 0.8963>
   rotation=<rot -88.4489 179.9968 0.0033>
  }]
}
constraints [
 {
  bone1Name=spine_04
  bone2Name=spine_03
  swing1LimitDegrees=45
  swing1Motion=Free
  swing2LimitDegrees=45
  swing2Motion=Free
  twistLimitDegrees=45
  twistMotion=Free
 }
 {
  bone1Name=spine_05
  bone2Name=spine_04
  swing1LimitDegrees=45
  swing1Motion=Free
  swing2LimitDegrees=45
  swing2Motion=Free
  twistLimitDegrees=45
  twistMotion=Free
 }
 {
  bone1Name=clavicle_l
  bone2Name=spine_05
  swing1LimitDegrees=45
  swing1Motion=Free
  swing2LimitDegrees=45
  swing2Motion=Free
  twistLimitDegrees=45
  twistMotion=Free
 }
 {
  bone1Name=upperarm_l
  bone2Name=clavicle_l
  swing1LimitDegrees=45
  swing1Motion=Free
  swing2LimitDegrees=45
  swing2Motion=Free
  twistLimitDegrees=45
  twistMotion=Free
 }
 {
  bone1Name=clavicle_r
  bone2Name=spine_05
  swing1LimitDegrees=45
  swing1Motion=Free
  swing2LimitDegrees=45
  swing2Motion=Free
  twistLimitDegrees=45
  twistMotion=Free
 }
 {
  bone1Name=upperarm_r
  bone2Name=clavicle_r
  swing1LimitDegrees=45
  swing1Motion=Free
  swing2LimitDegrees=45
  swing2Motion=Free
  twistLimitDegrees=45
  twistMotion=Free
 }
 {
  bone1Name=base_backHair_L_01
  bone2Name=spine_05
  swing1LimitDegrees=45
  swing1Motion=Free
  swing2LimitDegrees=45
  swing2Motion=Free
  twistLimitDegrees=45
  twistMotion=Free
 }
 {bone1Name=dyn_backHair_L_01, bone2Name=base_backHair_L_01, swing1LimitDegrees=0.5, swing2LimitDegrees=0.5, twistLimitDegrees=0.5}
 {bone1Name=dyn_backHair_L_02, bone2Name=dyn_backHair_L_01, swing1LimitDegrees=0.5, swing2LimitDegrees=0.5, twistLimitDegrees=0.5}
 {bone1Name=dyn_backHair_L_03, bone2Name=dyn_backHair_L_02, swing1LimitDegrees=0.5, swing2LimitDegrees=0.5, twistLimitDegrees=0.5}
 {
  bone1Name=base_backHair_M_01
  bone2Name=spine_05
  swing1LimitDegrees=45
  swing1Motion=Free
  swing2LimitDegrees=45
  swing2Motion=Free
  twistLimitDegrees=45
  twistMotion=Free
 }
 {bone1Name=dyn_backHair_M_01, bone2Name=base_backHair_M_01, swing1LimitDegrees=0.5, swing2LimitDegrees=0.5, twistLimitDegrees=0.5}
 {bone1Name=dyn_backHair_M_02, bone2Name=dyn_backHair_M_01, swing1LimitDegrees=0.5, swing2LimitDegrees=0.5, twistLimitDegrees=0.5}
 {bone1Name=dyn_backHair_M_03, bone2Name=dyn_backHair_M_02, swing1LimitDegrees=0.5, swing2LimitDegrees=0.5, twistLimitDegrees=0.5}
 {
  bone1Name=base_backHair_R_01
  bone2Name=spine_05
  swing1LimitDegrees=45
  swing1Motion=Free
  swing2LimitDegrees=45
  swing2Motion=Free
  twistLimitDegrees=45
  twistMotion=Free
 }
 {bone1Name=dyn_backHair_R_01, bone2Name=base_backHair_R_01, swing1LimitDegrees=0.5, swing2LimitDegrees=0.5, twistLimitDegrees=0.5}
 {bone1Name=dyn_backHair_R_02, bone2Name=dyn_backHair_R_01, swing1LimitDegrees=0.5, swing2LimitDegrees=0.5, twistLimitDegrees=0.5}
 {bone1Name=dyn_backHair_R_03, bone2Name=dyn_backHair_R_02, swing1LimitDegrees=0.5, swing2LimitDegrees=0.5, twistLimitDegrees=0.5}
 {
  bone1Name=dyn_backHair_R_02
  bone2Name=dyn_backHair_M_02
  swing1LimitDegrees=45
  swing1Motion=Locked
  swing2LimitDegrees=45
  swing2Motion=Locked
  twistLimitDegrees=45
  twistMotion=Locked
 }
 {
  bone1Name=dyn_backHair_L_02
  bone2Name=dyn_backHair_M_02
  swing1LimitDegrees=45
  swing1Motion=Locked
  swing2LimitDegrees=45
  swing2Motion=Locked
  twistLimitDegrees=45
  twistMotion=Locked
 }
 {bone1Name=dyn_backHair_R_corrective, bone2Name=dyn_backHair_R_01, swing1LimitDegrees=10, swing2LimitDegrees=10, twistLimitDegrees=2}
]
