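# Control Rig for the SprintLuckSweat female character's FaceAcc mesh, correcting front and back hair bones as the body moves.
# Parent constraints first blend hairMaster and front strands between neck, spine and clavicles, then a head-back corrective offsets hair by head rotation relative to spine_05.
# Spherical pose readers on clavicles and upper arms weight further additive hair offsets.
# A debug control-to-bone function is switched off; declared gameplay-state variables are unused.
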ControlRigBlueprint /BRCosmetics/Characters/Player/Female/Medium/Bodies/F_MED_SprintLuckSweat/Meshes/Parts/F_MED_SprintLuckSweat_FaceAcc_CtrlRig
graphs [
 CollapseNode_2_ContainedGraph {nodes [
   RigVMFunctionEntryNode()
   n1 = /Script/RigVM.RigVMFunction_MathQuaternionToRotator(Value=&RigVMFunction_MathTransformMakeRelative(
     Global=&RigUnit_GetRelativeTransformForItem(Child=(Type=Bone,Name="head"), Parent=(Type=Bone,Name="head"), bChildInitial=False, bParentInitial=true).RelativeTransform
     Parent=&RigUnit_GetRelativeTransformForItem(Child=(Type=Bone,Name="spine_05"), Parent=(Type=Bone,Name="spine_05"), bChildInitial=False, bParentInitial=true).RelativeTransform
    ).Local.Rotation) {is=struct RigVMUnitNode, template="Cast::Execute(in Value,out Result)"}
   > n2 = &RigUnit_ModifyTransforms(
    ItemToModify="((Item=(Type=Bone,Name=\"hairMaster\"),Transform=(Rotation=(X=0,Y=0,Z=0.1305,W=0.9914),Translation=(X=0,Y=-3,Z=0),Scale3D=(X=1,Y=1,Z=1))),(Item=(Type=Bone,Name=\"base_hair_Front_L_01\"),Transform=(Rotation=(X=0,Y=0,Z=0,W=1),Translation=(X=0,Y=0,Z=0),Scale3D=(X ...(truncated)"
    WeightMinimum=0
    Weight=&RigVMFunction_MathDoubleRemap(
     SourceMaximum=-55
     SourceMinimum=0
     TargetMaximum=1
     TargetMinimum=0
     Value=RigVMDispatchNode(Struct=@n1.Result) {template="DISPATCH_RigVMDispatch_BreakStruct(in Struct,out Elements)"}.Elements.Yaw
    ).Result
   )
   RigVMFunctionReturnNode(Completed=@n2.ExecutePin)
  ]}
 CollapseNode_ContainedGraph {nodes [
   n1 = RigVMFunctionEntryNode()
   &RigUnit_ParentConstraint(
    AdvancedSettings=(InterpolationType=Average,RotationOrderForFilter=XZY)
    Child=(Type=Bone,Name="hairMaster")
    Parents=((Item=(Type=Bone,Name="spine_05"),Weight=0.2),(Item=(Type=Bone,Name="neck_02"),Weight=1),(Item=(Type=Bone,Name="neck_01"),Weight=0))
    Weight=1
    ExecutePin=@n1.ExecutePin
   )
   > &RigUnit_ParentConstraint(
    AdvancedSettings=(InterpolationType=Average,RotationOrderForFilter=XZY)
    Child=(Type=Bone,Name="base_hair_Front_L_01")
    Parents=((Item=(Type=Bone,Name="head"),Weight=0),(Item=(Type=Bone,Name="clavicle_l"),Weight=0.5),(Item=(Type=Bone,Name="neck_01"),Weight=1),(Item=(Type=Bone,Name="spine_05"),Weight=0))
    Weight=1
    bMaintainOffset=true
   )
   > n2 = &RigUnit_ParentConstraint(
    AdvancedSettings=(InterpolationType=Average,RotationOrderForFilter=XZY)
    Child=(Type=Bone,Name="base_hair_Front_R_01")
    Parents=((Item=(Type=Bone,Name="head"),Weight=0),(Item=(Type=Bone,Name="clavicle_r"),Weight=0.5),(Item=(Type=Bone,Name="neck_01"),Weight=1),(Item=(Type=Bone,Name="spine_05"),Weight=0))
    Weight=1
   )
   RigVMFunctionReturnNode(ExecutePin=@n2.ExecutePin)
  ]}
 CollapseNode_ContainedGraph {nodes [
   n1 = RigVMFunctionEntryNode()
   n2 = &RigUnit_SphericalPoseReader(
    ActiveRegionScaleFactors=(PositiveWidth=0.5,NegativeWidth=3,PositiveHeight=1.5,NegativeHeight=1.5)
    ActiveRegionSize=0.1
    Debug=(bDrawDebug=false,bDraw2D=False,bDrawLocalAxes=False,DebugScale=25,DebugSegments=20,DebugThickness=0.25)
    DriverAxis=(X=-1,Y=0,Z=0)
    DriverItem=(Type=Bone,Name="upperarm_r")
    FalloffRegionScaleFactors=(PositiveWidth=1,NegativeWidth=1,PositiveHeight=1,NegativeHeight=1)
    FalloffSize=0.4
    FlipHeightScaling=False
    FlipWidthScaling=False
    OptionalParentItem=(Type=Bone,Name="None")
    RotationOffset=(X=90,Y=0,Z=180)
    ExecutePin=@n1.ExecutePin
   )
   > &RigUnit_ModifyTransforms(
    ItemToModify="((Item=(Type=Bone,Name=\"base_hair_Front_R_01\"),Transform=(Rotation=(X=0.0842,Y=0.2578,Z=-0.0226,W=0.9623),Translation=(X=0,Y=-5,Z=-4),Scale3D=(X=1,Y=1,Z=1))),(Item=(Type=Bone,Name=\"Dyn_hair_Front_R_03\"),Transform=(Rotation=(X=0,Y=0,Z=0,W=1),Translation=(X= ...(truncated)"
    WeightMinimum=0
    Weight=@n2.OutputParam
   )
   > n3 = &RigUnit_SphericalPoseReader(
    ActiveRegionScaleFactors=(PositiveWidth=0.5,NegativeWidth=3,PositiveHeight=1.5,NegativeHeight=1.5)
    ActiveRegionSize=0.1
    Debug=(bDrawDebug=false,bDraw2D=False,bDrawLocalAxes=False,DebugScale=25,DebugSegments=20,DebugThickness=0.25)
    DriverAxis=(X=1,Y=0,Z=0)
    DriverItem=(Type=Bone,Name="upperarm_l")
    FalloffRegionScaleFactors=(PositiveWidth=1,NegativeWidth=1,PositiveHeight=1,NegativeHeight=1)
    FalloffSize=0.4
    FlipHeightScaling=False
    FlipWidthScaling=False
    OptionalParentItem=(Type=Bone,Name="None")
    RotationOffset=(X=90,Y=0,Z=0)
   )
   > n4 = &RigUnit_ModifyTransforms(
    ItemToModify="((Item=(Type=Bone,Name=\"base_hair_Front_L_01\"),Transform=(Rotation=(X=0,Y=-0.0983,Z=0,W=0.9952),Translation=(X=0,Y=0,Z=0),Scale3D=(X=1,Y=1,Z=1))),(Item=(Type=Bone,Name=\"Dyn_hair_Front_L_02\"),Transform=(Rotation=(X=-0.0597,Y=0,Z=0,W=0.9982),Translation=(X=0 ...(truncated)"
    WeightMinimum=0
    Weight=@n3.OutputParam
   )
   RigVMFunctionReturnNode(ExecutePin=@n4.ExecutePin)
  ]}
 New Function_ContainedGraph {nodes [
   n1 = RigVMFunctionEntryNode()
   RigVMFunction_ControlFlowBranch = &RigVMFunction_ControlFlowBranch(Condition=@n1.Debug, ExecuteContext=@n1.ExecuteContext)
   RigVMFunctionReturnNode(ExecuteContext=@RigVMFunction_ControlFlowBranch.Completed)
   &RigUnit_SetTransform(
    Item=(Type=Bone,Name="clavicle_r")
    Weight=1
    bInitial=False
    Value=&RigUnit_GetTransform(Item=(Type=Control,Name="clavicle_r_ctrl"), bInitial=False).Transform
    ExecutePin=@RigVMFunction_ControlFlowBranch.True
   )
   RigUnit_GetTransform = &RigUnit_GetTransform(Item=(Type=Control,Name="upperarm_r_ctrl"), bInitial=False)
   > &RigUnit_SetTransform(Item=(Type=Bone,Name="upperarm_r"), Weight=1, bInitial=False, Value=@RigUnit_GetTransform.Transform)
   > RigUnit_SetTransform = &RigUnit_SetTransform(
    Item=(Type=Bone,Name="clavicle_l")
    Weight=1
    bInitial=False
    Value=&RigUnit_GetTransform(Item=(Type=Control,Name="clavicle_l_ctrl"), bInitial=False).Transform
   )
   > &RigUnit_SetTransform(
    Item=(Type=Bone,Name="upperarm_l")
    Weight=1
    bInitial=False
    Value=&RigUnit_GetTransform(Item=(Type=Control,Name="upperarm_l_ctrl"), bInitial=False).Transform
   )
   > &RigUnit_SetTransform(
    Item=(Type=Bone,Name="head")
    Weight=1
    bInitial=False
    Value=&RigUnit_GetTransform(Item=(Type=Control,Name="head_ctrl"), bInitial=False).Transform
   )
   > &RigUnit_SetTransform(
    Item=(Type=Bone,Name="neck_02")
    Weight=1
    bInitial=False
    Value=&RigUnit_GetTransform(Item=(Type=Control,Name="neck_02_ctrl"), bInitial=False).Transform
   )
   > &RigUnit_SetTransform(
    Item=(Type=Bone,Name="neck_01")
    Weight=1
    bInitial=False
    Value=&RigUnit_GetTransform(Item=(Type=Control,Name="neck_01_ctrl"), bInitial=False).Transform
   )
   > &RigUnit_SetTransform(
    Item=(Type=Bone,Name="spine_05")
    Weight=1
    bInitial=False
    Value=&RigUnit_GetTransform(Item=(Type=Control,Name="spine_05_ctrl"), bInitial=False).Transform
   )
  ]}
 New Function_ContainedGraph {nodes [
   n1 = RigVMFunctionEntryNode()
   RigVMFunction_ControlFlowBranch = &RigVMFunction_ControlFlowBranch(Condition=@n1.Debug, ExecuteContext=@n1.ExecuteContext)
   RigVMFunctionReturnNode(ExecuteContext=@RigVMFunction_ControlFlowBranch.Completed)
   DISPATCH_RigVMDispatch_ArrayIterator = RigVMDispatchNode(
    Array=/Script/ControlRig.RigUnit_CollectionGetAll(TypeToSearch=Control) {is=struct RigVMUnitNode}.Items
    ExecuteContext=@RigVMFunction_ControlFlowBranch.True
   ) {template="DISPATCH_RigVMDispatch_ArrayIterator(in Array,out Element,out Index,out Count,out Ratio)"}
   Replace = /Script/RigVM.RigVMFunction_NameReplace(Old=_ctrl, Name=@DISPATCH_RigVMDispatch_ArrayIterator.Element.Name) {is=struct RigVMUnitNode, template="Replace::Execute(in Name,in Old,in New,out Result)"}
   > &RigUnit_SetTransform(
    Item=(Type=Bone,Name="spine_01")
    Weight=1
    bInitial=False
    Item.Name=@Replace.Result
    Value=&RigUnit_GetTransform(Item=(Type=Control,Name="spine_01_ctrl"), bInitial=False, Item.Name=@DISPATCH_RigVMDispatch_ArrayIterator.Element.Name).Transform
   )
  ]}
 RigVMFunctionLibrary {nodes [
   Controls = RigVMCollapseNode(Debug=False, ExecuteContext=())
   ctrlRig = RigVMCollapseNode(Debug=False, ExecuteContext=())
  ]}
 RigVMFunction_Sequence_1_ContainedGraph {nodes [
   n1 = RigVMFunctionEntryNode()
   RigVMFunction_Sequence = &RigVMFunction_Sequence(ExecuteContext=@n1.ExecuteContext)
   n2 = &RigVMFunction_Sequence(ExecuteContext=@RigVMFunction_Sequence.B)
   RigVMFunctionReturnNode(A=@RigVMFunction_Sequence.A, B=@n2.A, C=@n2.B)
  ]}
 RigVMModel {nodes [
   RigUnit_BeginExecution = &RigUnit_BeginExecution()
   ctrlRig = RigVMFunctionReferenceNode(Debug=false, ExecuteContext=@RigUnit_BeginExecution.ExecutePin)
   > RigVMFunction_Sequence = RigVMAggregateNode()
   Parents Constraint = RigVMCollapseNode(ExecutePin=@RigVMFunction_Sequence.A)
   Head Back Corrective = RigVMCollapseNode(ExecuteContext=@"Parents Constraint".ExecutePin)
   n1 = &RigUnit_SphericalPoseReader(
    ActiveRegionScaleFactors=(PositiveWidth=2.4,NegativeWidth=3,PositiveHeight=3,NegativeHeight=3.5)
    ActiveRegionSize=0.1
    Debug=(bDrawDebug=false,bDraw2D=False,bDrawLocalAxes=False,DebugScale=25,DebugSegments=20,DebugThickness=0.25)
    DriverAxis=(X=1,Y=0,Z=0)
    DriverItem=(Type=Bone,Name="clavicle_l")
    FalloffRegionScaleFactors=(PositiveWidth=2,NegativeWidth=1,PositiveHeight=1,NegativeHeight=10)
    FalloffSize=0.4
    FlipHeightScaling=False
    FlipWidthScaling=False
    OptionalParentItem=(Type=Bone,Name="spine_01")
    RotationOffset=(X=80,Y=0,Z=0)
    ExecutePin=@"Head Back Corrective".Completed
   )
   > &RigUnit_ModifyTransforms(
    ItemToModify="((Item=(Type=Bone,Name=\"base_hair_Front_L_01\"),Transform=(Rotation=(X=0.0156,Y=0.1443,Z=-0.1064,W=0.9837),Translation=(X=0,Y=-1.1992,Z=-3.9099),Scale3D=(X=1,Y=1,Z=1))),(Item=(Type=Bone,Name=\"None\"),Transform=(Rotation=(X=0,Y=0,Z=0,W=1),Translation=(X=0,Y=0 ...(truncated)"
    WeightMinimum=0
    Weight=@n1.OutputParam
   )
   n2 = RigVMCollapseNode(ExecutePin=@RigVMFunction_Sequence.C)
   n3 = &RigUnit_SphericalPoseReader(
    ActiveRegionScaleFactors=(PositiveWidth=1,NegativeWidth=3,PositiveHeight=3,NegativeHeight=3)
    ActiveRegionSize=0.1
    Debug=(bDrawDebug=false,bDraw2D=False,bDrawLocalAxes=False,DebugScale=25,DebugSegments=20,DebugThickness=0.25)
    DriverAxis=(X=1,Y=0,Z=0)
    DriverItem=(Type=Bone,Name="clavicle_l")
    FalloffRegionScaleFactors=(PositiveWidth=0.6,NegativeWidth=1,PositiveHeight=1,NegativeHeight=1)
    FalloffSize=0.4
    FlipHeightScaling=False
    FlipWidthScaling=False
    OptionalParentItem=(Type=Bone,Name="spine_05")
    RotationOffset=(X=0,Y=-35,Z=10)
    ExecutePin=@n2.ExecutePin
   )
   > &RigUnit_ModifyTransforms(
    ItemToModify=((Item=(Type=Bone,Name="base_backHair_L_01"),Transform=(Rotation=(X=0,Y=0,Z=0,W=1),Translation=(X=0,Y=-3,Z=-5),Scale3D=(X=1,Y=1,Z=1))))
    WeightMinimum=0
    Weight=@n3.OutputParam
   )
   > &RigUnit_ModifyTransforms(
    ItemToModify=((Item=(Type=Bone,Name="base_backHair_M_01"),Transform=(Rotation=(X=0,Y=0,Z=0,W=1),Translation=(X=0,Y=-2,Z=-2),Scale3D=(X=1,Y=1,Z=1))))
    WeightMinimum=0
    Weight=@n3.OutputParam
   )
   > n4 = &RigUnit_SphericalPoseReader(
    ActiveRegionScaleFactors=(PositiveWidth=1,NegativeWidth=3,PositiveHeight=3,NegativeHeight=3)
    ActiveRegionSize=0.1
    Debug=(bDrawDebug=false,bDraw2D=False,bDrawLocalAxes=False,DebugScale=25,DebugSegments=20,DebugThickness=0.25)
    DriverAxis=(X=-1,Y=0,Z=0)
    DriverItem=(Type=Bone,Name="clavicle_r")
    FalloffRegionScaleFactors=(PositiveWidth=0.6,NegativeWidth=1,PositiveHeight=1,NegativeHeight=1)
    FalloffSize=0.4
    FlipHeightScaling=False
    FlipWidthScaling=False
    OptionalParentItem=(Type=Bone,Name="spine_05")
    RotationOffset=(X=0,Y=145,Z=10)
   )
   > &RigUnit_ModifyTransforms(
    ItemToModify="((Item=(Type=Bone,Name=\"base_backHair_R_01\"),Transform=(Rotation=(X=0,Y=0,Z=0,W=1),Translation=(X=0,Y=-3,Z=1),Scale3D=(X=1,Y=1,Z=1))),(Item=(Type=Bone,Name=\"base_backHair_R_01\"),Transform=(Rotation=(X=0,Y=0,Z=0,W=1),Translation=(X=0,Y=0,Z=0),Scale3D=(X=1,Y ...(truncated)"
    WeightMinimum=0
    Weight=@n4.OutputParam
   )
   > n5 = &RigUnit_SphericalPoseReader(
    ActiveRegionScaleFactors=(PositiveWidth=1,NegativeWidth=3,PositiveHeight=3,NegativeHeight=3)
    ActiveRegionSize=0.1
    Debug=(bDrawDebug=false,bDraw2D=False,bDrawLocalAxes=False,DebugScale=25,DebugSegments=20,DebugThickness=0.25)
    DriverAxis=(X=1,Y=0,Z=0)
    DriverItem=(Type=Bone,Name="clavicle_l")
    FalloffRegionScaleFactors=(PositiveWidth=0.6,NegativeWidth=1,PositiveHeight=1,NegativeHeight=1)
    FalloffSize=0.4
    FlipHeightScaling=false
    FlipWidthScaling=false
    OptionalParentItem=(Type=Bone,Name="spine_05")
    RotationOffset=(X=-45,Y=0,Z=0)
   )
   > &RigUnit_ModifyTransforms(
    ItemToModify=((Item=(Type=Bone,Name="base_backHair_L_01"),Transform=(Rotation=(X=0.2588,Y=0,Z=0,W=0.9659),Translation=(X=0,Y=-2,Z=0),Scale3D=(X=1,Y=1,Z=1))))
    WeightMinimum=0
    Weight=@n5.OutputParam
   )
  ]}
]
hierarchy [
 root {loc=<0 0 0>, rot=<0 0 0 1>}
 pelvis {parent=root, loc=<0 -0.5301 98.0571>, rot=<0.0127 0.707 -0.0127 -0.707>}
 spine_01 {parent=pelvis, loc=<6.1208 -0.9328 0>, rot=<0 0 0.034 -0.9994>}
 spine_02 {parent=spine_01, loc=<6.5997 -0.2095 0>, rot=<0 0 0.0238 -0.9997>}
 spine_03 {parent=spine_02, loc=<6.4914 -0.1237 0>, rot=<0 0 0.0149 -0.9999>}
 spine_04 {parent=spine_03, loc=<6.9598 0.0794 0>, rot=<0 0 -0.0132 -0.9999>}
 spine_05 {parent=spine_04, loc=<6.9987 0 0>, rot=<0 0 -0.0015 -1>}
 neck_01 {parent=spine_05, loc=<11.3138 0.1529 0>, rot=<0 0 -0.2003 -0.9797>}
 neck_02 {parent=neck_01, loc=<4.5627 0 0>, rot=<0 0 0.0504 -0.9987>}
 head {parent=neck_02, loc=<5.3914 0 0>, rot=<0 0 0.115 -0.9934>}
 hairMaster {parent=head, loc=<7.3274 -4.4725 0.003>, rot=<0.1794 -0.9838 -0.0001 -0.0007>}
 base_backHair_L_01 {parent=hairMaster, loc=<7.819 2.1314 7.4979>, rot=<0.0001 0 0.0426 -0.9991>}
 dyn_backHair_L_01 {parent=base_backHair_L_01, loc=<0 0 0>, rot=<0 0 0 1>}
 dyn_backHair_L_02 {parent=dyn_backHair_L_01, loc=<10.707 0 0>, rot=<-0.0001 -0.0001 -0.0963 -0.9954>}
 dyn_backHair_L_03 {parent=dyn_backHair_L_02, loc=<10.1131 0 0>, rot=<0 0 0 1>}
 base_backHair_R_01 {parent=hairMaster, loc=<7.819 2.1167 -7.5021>, rot=<0.0001 0.0001 0.0618 -0.9981>}
 dyn_backHair_R_01 {parent=base_backHair_R_01, loc=<0 0 0>, rot=<0 0 0 1>}
 dyn_backHair_R_02 {parent=dyn_backHair_R_01, loc=<10.9095 0 0>, rot=<-0.0002 -0.0001 -0.1261 -0.992>}
 dyn_backHair_R_03 {parent=dyn_backHair_R_02, loc=<9.9777 0 0>, rot=<0 0 0 1>}
 dyn_backHair_R_corrective {parent=dyn_backHair_R_01, loc=<5.0226 4.5219 -2.4497>, rot=<0.0004 -0.0001 -0.1261 -0.992>}
 base_backHair_M_01 {parent=hairMaster, loc=<8.5957 0 0>, rot=<0 0 -0.016 -0.9999>}
 dyn_backHair_M_01 {parent=base_backHair_M_01, loc=<0 0 0>, rot=<0 0 0 1>}
 dyn_backHair_M_02 {parent=dyn_backHair_M_01, loc=<11.0334 0 0>, rot=<-0.0001 -0.0001 -0.0748 -0.9972>}
 dyn_backHair_M_03 {parent=dyn_backHair_M_02, loc=<10.1531 0 0>, rot=<0 0 0 1>}
 base_hair_Front_R_01 {parent=head, loc=<-2.0335 1.4519 7.1304>, rot=<0.0786 -0.0296 0.9896 0.1168>}
 Dyn_hair_Front_R_01 {parent=base_hair_Front_R_01, loc=<0 0 0>, rot=<0 0 0 1>}
 Dyn_hair_Front_R_02 {parent=Dyn_hair_Front_R_01, loc=<6.2483 0 0>, rot=<0 -0.1243 0.0883 -0.9883>}
 Dyn_hair_Front_R_03 {parent=Dyn_hair_Front_R_02, loc=<5.262 0 0>, rot=<0 -0.077 0.1244 -0.9892>}
 Dyn_hair_Front_R_04 {parent=Dyn_hair_Front_R_03, loc=<5.1351 0 0>, rot=<0 0 0 1>}
 base_hair_Front_L_01 {parent=head, loc=<-1.7516 -0.6259 -7.6829>, rot=<-0.0132 0.0363 0.9781 0.2045>}
 Dyn_hair_Front_L_01 {parent=base_hair_Front_L_01, loc=<0 0 0>, rot=<0.0011 0.0031 -0.1284 -0.9917>}
 Dyn_hair_Front_L_02 {parent=Dyn_hair_Front_L_01, loc=<5.9848 0 0>, rot=<-0.0217 0.0696 0.2961 -0.9524>}
 Dyn_hair_Front_L_03 {parent=Dyn_hair_Front_L_02, loc=<5.8155 0 0>, rot=<0 -0.0107 -0.0028 -0.9999>}
 Dyn_hair_Front_L_04 {parent=Dyn_hair_Front_L_03, loc=<5.9511 0 0>, rot=<0 0 0 1>}
 neck_01_mover {parent=neck_01, loc=<0 0 0>, rot=<0 0 0 1>}
 clavicle_l {parent=spine_05, loc=<5.4831 -0.1053 -3.2946>, rot=<0.0465 -0.7289 -0.009 -0.6829>}
 upperarm_l {parent=clavicle_l, loc=<9.8658 -1.2136 -0.4105>, rot=<0.0164 -0.3245 -0.0269 -0.9454>}
 clavicle_r {parent=spine_05, loc=<5.4831 -0.1053 3.2946>, rot=<0.7289 0.0465 -0.6829 0.009>}
 upperarm_r {parent=clavicle_r, loc=<-9.8658 1.2136 0.4105>, rot=<0.0164 -0.3245 -0.0269 -0.9454>}
 neck_01_ctrl {controlType=EULER_TRANSFORM, type=Control, loc=<0 0 0>, rot=<0 0 0 1>}
 neck_02_ctrl {controlType=EULER_TRANSFORM, parent=neck_01_ctrl, type=Control, loc=<0 0 0>, rot=<0 0 0 1>}
 head_ctrl {controlType=EULER_TRANSFORM, parent=neck_02_ctrl, type=Control, loc=<0 0 0>, rot=<0 0 0 1>}
 spine_05_ctrl {controlType=EULER_TRANSFORM, type=Control, loc=<0 0 0>, rot=<0 0 0 1>}
 clavicle_l_ctrl {controlType=EULER_TRANSFORM, parent=spine_05_ctrl, type=Control, loc=<0 0 0>, rot=<0 0 0 1>}
 upperarm_l_ctrl {controlType=EULER_TRANSFORM, parent=clavicle_l_ctrl, type=Control, loc=<0 0 0>, rot=<0 0 0 1>}
 clavicle_r_ctrl {controlType=EULER_TRANSFORM, parent=spine_05_ctrl, type=Control, loc=<0 0 0>, rot=<0 0 0 1>}
 upperarm_r_ctrl {controlType=EULER_TRANSFORM, type=Control, loc=<0 0 0>, rot=<0 0 0 1>}
]
variables [
 Is Wearing Backpack or Cape Alpha {type=double, default=0}
 Parachute Open Alpha {type=double, default=0}
 is Parachute Open {type=bool, default=False}
 is Playing Emote {type=bool, default=False}
 is Sky Floating {type=bool, default=False}
 is Skydive Dive {type=bool, default=False}
]
