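# Physics asset for the front hair of the F_MED_SprintLuckSweat female character cosmetic, with capsule bodies for two hair strands and the upper body.
# Left and right four-segment dyn_hair chains hang from small base bodies on the head, with tight swing limits and almost no twist.
# Spine, neck, head, clavicle and upper-arm capsules, plus a long neck_01_mover capsule, form the body volume around the hair, linked by free angular constraints.
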PhysicsAsset /BRCosmetics/Characters/Player/Female/Medium/Bodies/F_MED_SprintLuckSweat/Meshes/Parts/F_MED_SprintLuckSweat_FaceAcc_FrontHair_Physics1
bodies {
 base_hair_Front_L_01 [{shapeType=Capsule, length=0.505, radius=0.505, shapeName=base_hair_Front_L_01_capsule, rotation=<rot 0 -90 0>}]
 base_hair_Front_R_01 [{shapeType=Capsule, length=0.505, radius=0.505, shapeName=base_hair_Front_R_01_capsule, rotation=<rot 0 -90 0>}]
 clavicle_l [{
   shapeType=Capsule
   length=8.3762
   radius=6.6217
   shapeName=clavicle_l_capsule
   center=<3.745 -1.0162 0.1235>
   rotation=<rot 77.3884 3.5917 -1.6432>
  }]
 clavicle_r [{
   shapeType=Capsule
   length=8.7705
   radius=6.2273
   shapeName=clavicle_l_capsule
   center=<-5.0595 1.0086 -0.2549>
   rotation=<rot 76.1657 5.0603 2.0442>
  }]
 dyn_hair_Front_L_01 [{shapeType=Capsule, length=1.8828, radius=3.1248, shapeName=dyn_hair_Front_L_01_capsule, center=<3.4545 0.7875 -0.8294>}]
 dyn_hair_Front_L_02 [{shapeType=Capsule, length=0.6967, radius=3.4662, shapeName=dyn_hair_Front_L_02_capsule, center=<2.991 -2.0269 0>}]
 dyn_hair_Front_L_03 [{shapeType=Capsule, length=4.506, radius=2.752, shapeName=dyn_hair_Front_L_03_capsule, center=<2.991 -1.3796 0>}]
 dyn_hair_Front_L_04 [{shapeType=Capsule, length=2.8973, radius=2.7806, shapeName=dyn_hair_Front_L_04_capsule, center=<2.0928 0 0>}]
 dyn_hair_Front_R_01 [{shapeType=Capsule, length=1.0641, radius=3.091, shapeName=dyn_hair_Front_R_01_capsule, center=<2.9115 0.7619 1.1475>}]
 dyn_hair_Front_R_02 [{shapeType=Capsule, length=2.8238, radius=3.4966, shapeName=dyn_hair_Front_R_02_capsule, center=<2.991 -0.8519 0>}]
 dyn_hair_Front_R_03 [{shapeType=Capsule, length=7.3541, radius=2.7077, shapeName=dyn_hair_Front_R_03_capsule, center=<2.991 -0.8619 0>}]
 dyn_hair_Front_R_04 [{shapeType=Capsule, length=5.417, radius=2.3991, shapeName=dyn_hair_Front_R_04_capsule, center=<2.432 -0.2299 0>}]
 head [{shapeType=Capsule, length=5.0814, radius=5.7588, shapeName=head_capsule, center=<1.7057 3.6001 0>, rotation=<rot -90 9.9999 0>}]
 neck_01 [{shapeType=Capsule, length=1.782, radius=4.5005, shapeName=neck_01_capsule, center=<3.2157 0 0>, rotation=<rot -90 0 0>}]
 neck_01_mover [{shapeType=Capsule, length=51.0221, radius=5.5914, shapeName=neck_01_capsule, center=<-3.8131 -0.154 0>, rotation=<rot 0 -21.3482 0>}]
 spine_04 [{shapeType=Capsule, length=4.4467, radius=11.0749, shapeName=spine_04_capsule, center=<0 2.6192 0>}]
 spine_05 [{shapeType=Capsule, length=10.7087, radius=11.2316, shapeName=spine_05_capsule, center=<-1.594 -1.1481 0>}]
 upperarm_l [{
   shapeType=Capsule
   length=24.2645
   radius=5.7355
   shapeName=upperarm_l_capsule
   center=<12.6605 -0.1968 -0.0207>
   rotation=<rot -86.3491 0.0015 -0.0014>
  }]
 upperarm_r [{
   shapeType=Capsule
   length=23.9246
   radius=6.0754
   shapeName=upperarm_l_capsule
   center=<-12.6605 0.1968 0.0207>
   rotation=<rot -86.3028 0.0014 179.9986>
  }]
}
constraints [
 {
  bone1Name=spine_05
  bone2Name=spine_04
  swing1LimitDegrees=45
  swing1Motion=Free
  swing2LimitDegrees=45
  swing2Motion=Free
  twistLimitDegrees=45
  twistMotion=Free
 }
 {
  bone1Name=clavicle_l
  bone2Name=spine_05
  swing1LimitDegrees=45
  swing1Motion=Free
  swing2LimitDegrees=45
  swing2Motion=Free
  twistLimitDegrees=45
  twistMotion=Free
 }
 {
  bone1Name=upperarm_l
  bone2Name=clavicle_l
  swing1LimitDegrees=45
  swing1Motion=Free
  swing2LimitDegrees=45
  swing2Motion=Free
  twistLimitDegrees=45
  twistMotion=Free
 }
 {
  bone1Name=clavicle_r
  bone2Name=spine_05
  swing1LimitDegrees=45
  swing1Motion=Free
  swing2LimitDegrees=45
  swing2Motion=Free
  twistLimitDegrees=45
  twistMotion=Free
 }
 {
  bone1Name=upperarm_r
  bone2Name=clavicle_r
  swing1LimitDegrees=45
  swing1Motion=Free
  swing2LimitDegrees=45
  swing2Motion=Free
  twistLimitDegrees=45
  twistMotion=Free
 }
 {
  bone1Name=base_hair_Front_L_01
  bone2Name=head
  swing1LimitDegrees=45
  swing1Motion=Free
  swing2LimitDegrees=45
  swing2Motion=Free
  twistLimitDegrees=45
  twistMotion=Free
 }
 {bone1Name=dyn_hair_Front_L_01, bone2Name=base_hair_Front_L_01, swing1LimitDegrees=30, swing2LimitDegrees=10, twistLimitDegrees=1}
 {bone1Name=dyn_hair_Front_L_02, bone2Name=dyn_hair_Front_L_01, swing1LimitDegrees=20, swing2LimitDegrees=10, twistLimitDegrees=1}
 {bone1Name=dyn_hair_Front_L_03, bone2Name=dyn_hair_Front_L_02, swing1LimitDegrees=20, swing2LimitDegrees=10, twistLimitDegrees=1}
 {bone1Name=dyn_hair_Front_L_04, bone2Name=dyn_hair_Front_L_03, swing1LimitDegrees=20, swing2LimitDegrees=10, twistLimitDegrees=1}
 {
  bone1Name=base_hair_Front_R_01
  bone2Name=head
  swing1LimitDegrees=45
  swing1Motion=Free
  swing2LimitDegrees=45
  swing2Motion=Free
  twistLimitDegrees=45
  twistMotion=Free
 }
 {bone1Name=dyn_hair_Front_R_01, bone2Name=base_hair_Front_R_01, swing1LimitDegrees=30, swing2LimitDegrees=10, twistLimitDegrees=1}
 {bone1Name=dyn_hair_Front_R_02, bone2Name=dyn_hair_Front_R_01, swing1LimitDegrees=20, swing2LimitDegrees=10, twistLimitDegrees=1}
 {bone1Name=dyn_hair_Front_R_03, bone2Name=dyn_hair_Front_R_02, swing1LimitDegrees=20, swing2LimitDegrees=10, twistLimitDegrees=1}
 {bone1Name=dyn_hair_Front_R_04, bone2Name=dyn_hair_Front_R_03, swing1LimitDegrees=20, swing2LimitDegrees=10, twistLimitDegrees=1}
 {
  bone1Name=neck_01_mover
  bone2Name=neck_01
  swing1LimitDegrees=45
  swing1Motion=Free
  swing2LimitDegrees=45
  swing2Motion=Free
  twistLimitDegrees=45
  twistMotion=Free
 }
 {
  bone1Name=neck_01
  bone2Name=spine_05
  swing1LimitDegrees=45
  swing1Motion=Free
  swing2LimitDegrees=45
  swing2Motion=Free
  twistLimitDegrees=45
  twistMotion=Free
 }
 {
  bone1Name=head
  bone2Name=neck_01
  swing1LimitDegrees=45
  swing1Motion=Free
  swing2LimitDegrees=45
  swing2Motion=Free
  twistLimitDegrees=45
  twistMotion=Free
 }
]
